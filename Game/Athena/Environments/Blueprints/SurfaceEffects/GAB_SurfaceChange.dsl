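; Gameplay ability that applies and removes surface effects on the player pawn for ice, snow, lava, sand, chrome and cube.
; Surface-change events switch on surface tag, run SurfaceCleanup to strip other surfaces' effects, then apply that surface's effect.
; Lava launches the pawn with seeded random angle and a growing multiplier; movement-mode changes toggle its periodic damage.
; Cleanup also runs on aircraft phase, and a short suppress effect masks unpredicted ice removal.
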(event Custom|OnSurfaceChanged
  (params (EventData :type Gameplay%20Event%20Data%20Structure%20%28by%20ref%29))
  (bind (_eventtag _instigator _target _optionalobject _optionalobject2 _contexthandle _instigatortags) (Utilities|Struct|BreakGameplayEventData EventData))
  (bind _asfort_pawn (Utilities|Casting|CastToFortPawn (Ability|GetAvatarActorfromActorInfo)))
  (bind _output_get (Variables|Default|SetAvatarPawn _asfort_pawn))
  (Character|BindEventtoMovementModeChangedDelegate _output_get (EventDispatchers|CreateEvent :!event Custom|MovementModeChanged))
  (CallFunction|RegisterGamePhaseListener)
  (Development|PrintString (GameplayTags|GetDebugStringfromGameplayTagContainer _instigatortags) (Variables|Default|GetDebug) (Variables|Default|GetDebug))
  (Utilities|FlowControl|Sequence
    (:then_0
      (switch Utilities|FlowControl|Switch|SwitchonGameplayTagContainer _instigatortags
        (:Default
          ;[1 1 1] Default, so remove all GEs on a delay
          (CallFunction|SurfaceCleanup true))
        (:(GameplayTags=((TagName="Fort.SurfaceType.Ice")))
          ;[1 0 0] Ice provides slippery movement, neeeds better removal solution before we can add other gameplay affecting GEs
          (CallFunction|SurfaceCleanup false "SurfaceType24")
          (CallFunction|HotfixableGEApplication (Variables|Default|GetIcySurfacesEnabled) (Variables|Default|GetGE_SurfaceChange_Ice) (Variables|Default|GetGE_SurfaceChange_Snow))
          ;[1 1 1 0.5] Removes Icy Feet when player enters vehicle to fix FORT-1047060
          (Ability|Tasks|WaitGameplayTagAdd "(TagName=\"Gameplay.Action.Player.AthenaInVehicle\")"
            (:then
              ;; Removes Icy Feet on DBNO to fix FORT-1125903
              (Ability|Tasks|WaitGameplayTagAdd "(TagName=\"Gameplay.Status.DBNO\")"
                (:Added
                  _)))
            (:Added
              _))
          (CallFunction|SurfaceCleanup false)
          (Variables|Default|SetPreviousSurfaceTagContainer)
          (Ability|RemoveGameplayEffectFromOwnerWithAssetTags "(GameplayTags=((TagName=\"GameplayEffect.Surface.Ice\")))"))
        (:(GameplayTags=((TagName="Fort.SurfaceType.Snow")))
          ;[1 1 1] Snow has GCs only, so we remove it on a delay
          (CallFunction|SurfaceCleanup true "SurfaceType23")
          (Ability|ApplyGameplayEffectToOwner "/Game/Athena/Environments/Blueprints/SurfaceEffects/GE_SurfaceChange_Snow.GE_SurfaceChange_Snow_C" 0))
        (:(GameplayTags=((TagName="Fort.SurfaceType.Lava")))
          (CallFunction|SurfaceCleanup false "SurfaceType25")
          (Ability|ApplyGameplayEffectToOwner "/Game/Athena/Environments/Blueprints/SurfaceEffects/GE_SurfaceChange_Lava.GE_SurfaceChange_Lava_C" 0)
          (CallFunction|LavaBounce))
        (:(GameplayTags=((TagName="Fort.SurfaceType.Sand")))
          (CallFunction|SurfaceCleanup false "SurfaceType26")
          (Ability|ApplyGameplayEffectToOwner "/Game/Athena/Environments/Blueprints/SurfaceEffects/Sand/GE_SurfaceChange_Sand_Surface.GE_SurfaceChange_Sand_Surface_C" 0))
        (:(GameplayTags=((TagName="Fort.SurfaceType.Cube")))
          ;[1 0.2904 0.7663] This checks to see if the pawn is stepping onto a cube, if they are it launches the pawn into the air a tiny bit so it will trigger the Hit event on the Cube, allowing its logic to take over.
          (bind _returnvalue (GameplayTags|HasMatchingGameplayTag(Message) _optionalobject "(TagName=\"Granted.Athena.Cube.Actor\")"))
          (if _returnvalue
            (Utilities|FlowControl|Delay 0.1)
            (if (AI|Components|NavMovement|IsFalling (Variables|/Script/Engine.Character|GetCharacterMovement (Variables|Default|GetAvatarPawn)))
              (CallFunction|SurfaceCleanup true)
              (else
                (Pawn|Character|LaunchCharacterJump (Variables|Default|GetAvatarPawn) "0, 0,10.000000" false true false true)))
            (bind _asactor (Utilities|Casting|CastToActor _optionalobject))
            (Ability|SendGameplayEventtoActor _asactor "(TagName=\"GameplayEvent.SurfaceChange.Cube\")" (Utilities|Struct|MakeGameplayEventData "(TagName=\"GameplayEvent.SurfaceChange.Cube\")" (Variables|Default|GetAvatarPawn) (Variables|Default|GetAvatarPawn)))
            (else
              _)))
        (:(GameplayTags=((TagName="Fort.SurfaceType.Chrome")))
          (CallFunction|SurfaceCleanup false "SurfaceType28")
          (Ability|ApplyGameplayEffectToOwner "/Game/Athena/Environments/Blueprints/SurfaceEffects/Sand/GE_SurfaceChange_Chrome_Surface.GE_SurfaceChange_Chrome_Surface_C" 0))))
    (:then_1
      (Variables|Default|SetPreviousSurfaceTagContainer _instigatortags))))

(event Ability|EventOnEndAbility
  (params (bWasCancelled :type Boolean))
  (bind _gamephaseupdatedasynctask (Variables|Default|GetGamePhaseUpdatedAsyncTask)
    (:then
      (GameplayEventRouter|StopListeningtoStatefulEvent _gamephaseupdatedasynctask))
    (:else
      _))
  (bind _async_task (Variables|Default|GetAsync%20Task)
    (:then
      (GameplayTasks|EndTask _async_task))
    (:else)))

(event Custom|RegisterGamePhaseListener
  (Utilities|IsValid (Variables|Default|GetGamePhaseUpdatedAsyncTask)
    (:Is%20Not%20Valid
      (bind _returnvalue (GameplayEventRouter|GetEventRouter "/Script/FortniteGame.FortGameState" (Variables|Default|GetAvatarPawn)))
      (bind (_asynctaskproxy _payload) (GameplayEventRouter|StartListeningToStatefulEvent _returnvalue "/Script/FortniteGame.FortGamePhaseUpdatedEvent")
        (:then
          (Variables|Default|SetGamePhaseUpdatedAsyncTask _asynctaskproxy))
        (:OnEventReceived
          _)
        (:OnSavedEventStateExists
          _))
      (CallFunction|GamePhaseChanged (Utilities|Struct|BreakFortGamePhaseUpdatedEvent _payload)))))

(event Custom|MovementModeChanged
  (params (Character :type Character%20Object%20Reference) (PrevMovementMode :type EMovementMode%20Enum :default "MOVE_None") (PreviousCustomMode :type Byte))
  (bind _charactermovement (Variables|/Script/Engine.Character|GetCharacterMovement Character))
  (Utilities|IsValid Character
    (:Is%20Valid
      (Utilities|FlowControl|Sequence
        (:then_0
          ;[1 1 1] Handle applying/removing the lava periodic damage GE depending on if grounded, if touching lava, and if already applied
          (bind _asfort_player_pawn (Utilities|Casting|CastToFortPlayerPawn Character))
          (bind _asfort_movement_comp_character (Utilities|Casting|CastToFortMovementComp_Character (Variables|/Script/Engine.Character|GetCharacterMovement Character)))
          (CallFunction|HandleLavaPeriodicDamageGE _asfort_movement_comp_character (Footsteps|GetSurfaceTrackingComponent _asfort_player_pawn)))
        (:then_1
          (Utilities|IsValid _charactermovement
            (:Is%20Valid
              (switch Utilities|FlowControl|Switch|SwitchonEMovementMode (Variables|/Script/Engine.CharacterMovementComponent|GetMovementMode _charactermovement)
                (:MOVE_Flying
                  (if (GameplayTags|HasMatchingGameplayTag (Variables|Default|GetAvatarPawn) "(TagName=\"Gameplay.Action.Player.Creative.Flying\")")
                    (CallFunction|SurfaceCleanup false)))))))))))

(event Custom|GamePhaseChanged
  (params (GamePhase :type EAthenaGamePhase%20Enum))
  ;[1 1 1] Clean up all effects that may be on the player when we put them on the battle bus
  (if (Utilities|Enum|Equal(Enum) GamePhase "Aircraft")
    (CallFunction|SurfaceCleanup false)))

(event Ability|EventActivateAbility
  (bind _async_task (Variables|Default|GetAsync%20Task))
  (Utilities|IsValid _async_task
    (:Is%20Valid
      (GameplayTasks|EndTask _async_task))
    (:Is%20Not%20Valid
      _))
  (bind (_asynctaskproxy _payload) (Ability|Tasks|WaitGameplayEvent "(TagName=\"Event.SurfaceType.Changed\")" 0 false false)
    (:then
      (Variables|Default|SetAsync%20Task _asynctaskproxy))
    (:EventReceived
      (CallFunction|OnSurfaceChanged _payload))))

(fn HandleLavaPeriodicDamageGE
  (params (FortMovementComp :type Fort%20Movement%20Comp%20Character%20Object%20Reference) (FortSurfaceTrackingComp :type Fort%20Surface%20Tracking%20Component%20Object%20Reference))
  (if (CallFunction|ShouldApplyLavaPeriodicDamage FortMovementComp FortSurfaceTrackingComp)
    (CallFunction|HotfixableGEApplication (Variables|ScalableFloats|GetLavaPeriodicDamageEnabled) "/Game/Athena/Environments/Blueprints/SurfaceEffects/GE_SurfaceChange_Lava_Damage_Periodic.GE_SurfaceChange_Lava_Damage_Periodic_C")
    (elif (CallFunction|ShouldRemoveLavaPeriodicDamage FortMovementComp FortSurfaceTrackingComp)
      (Ability|RemoveGameplayEffectFromOwnerWithGrantedTags (Variables|Default|GetLavaPeriodicDamageTags))
      (else
        _)))
  (return))

(fn HotfixableGEApplication
  (params (Input :type Scalable%20Float%20Structure%20%28by%20ref%29) (GameplayEffectAppliedOnTrue :type Gameplay%20Effect%20Class%20Reference) (GameplayEffectAppliedOnFalse :type Gameplay%20Effect%20Class%20Reference))
  (if (ScalableFloat|GetValueasBool Input)
    (bind _returnvalue (Ability|ApplyGameplayEffectToOwner GameplayEffectAppliedOnTrue 0))
    (return _returnvalue)
    (elif (Utilities|IsValidClass GameplayEffectAppliedOnFalse)
      (bind _returnvalue_2 (Ability|ApplyGameplayEffectToOwner GameplayEffectAppliedOnFalse 0))
      (return _returnvalue_2)
      (else
        (return 0)))))

(fn LavaBounce
  (bind _baseverticalbouncevelocity (Variables|ScalableFloats|GetBaseVerticalBounceVelocity))
  (bind _returnvalue (ScalableFloat|GetValueatLevel _baseverticalbouncevelocity))
  (bind _returnvalue_2 (* "0.000000,0.000000,1.000000" _returnvalue))
  (bind _returnvalue_3 (Ability|GetAvatarActorfromActorInfo))
  (bind _baselateralbouncevelocity (Variables|ScalableFloats|GetBaseLateralBounceVelocity))
  (bind _returnvalue_4 (ScalableFloat|GetValueatLevel _baselateralbouncevelocity))
  (bind _lavabouncemultiplier (Variables|Default|GetLavaBounceMultiplier))
  (bind _returnvalue_5 (* _lavabouncemultiplier _returnvalue_4))
  (bind _facinganglevelocitycomponent (Variables|ScalableFloats|GetFacingAngleVelocityComponent))
  (bind _returnvalue_6 (ScalableFloat|GetValueatLevel _facinganglevelocitycomponent))
  (bind _randomangleoffset (Variables|Default|GetRandomAngleOffset))
  (bind _returnvalue_7 (Math|Rotator|MakeRotator 0.0 0.0 _randomangleoffset))
  (bind _avatarpawn (Variables|Default|GetAvatarPawn))
  (bind _returnvalue_8 (Transformation|GetActorRotation _avatarpawn))
  (bind _returnvalue_9 (Math|Rotator|GetRotationXVector _returnvalue_8))
  (bind _returnvalue_10 (Math|Vector|RotateVector _returnvalue_9 _returnvalue_7))
  (bind _returnvalue_11 (Math|Conversions|ToVector2D(Vector) _returnvalue_10))
  (bind _returnvalue_12 (Math|Vector2D|Normalize2D _returnvalue_11))
  (bind _returnvalue_13 (* _returnvalue_12 _returnvalue_6))
  (bind _returnvalue_14 (- 1.0 _returnvalue_6))
  (bind _returnvalue_15 (Transformation|GetVelocity _returnvalue_3))
  (bind _returnvalue_16 (Math|Vector|MirrorVectorbyNormal _returnvalue_15 "0.000000, 0,1.000000"))
  (bind _returnvalue_17 (Math|Conversions|ToVector2D(Vector) _returnvalue_16))
  (bind _returnvalue_18 (Math|Vector2D|Normalize2D _returnvalue_17))
  (bind _returnvalue_19 (* _returnvalue_18 _returnvalue_14))
  (bind _returnvalue_20 (+ _returnvalue_19 _returnvalue_13))
  (bind _returnvalue_21 (Math|Conversions|ToVector(Vector2D) _returnvalue_20))
  (bind _returnvalue_22 (Math|Vector|Normalize _returnvalue_21))
  (bind _returnvalue_23 (* _returnvalue_22 _returnvalue_5))
  (bind (_returnvalue_roll _returnvalue_pitch _returnvalue_yaw) (Math|Rotator|BreakRotator (Transformation|GetActorRotation (Variables|Default|GetAvatarPawn))))
  (bind (_returnvalue_x _returnvalue_y) (Math|Vector|BreakVector (Transformation|GetActorLocation (Variables|Default|GetAvatarPawn))))
  (bind _returnvalue_24 (+ _returnvalue_2 _returnvalue_23))
  ;[1 1 1] Apply GEs for both instant and periodic damage
  (Ability|ApplyGameplayEffectToOwner "/Game/Athena/Environments/Blueprints/SurfaceEffects/GE_SurfaceChange_Lava_Damage.GE_SurfaceChange_Lava_Damage_C" 0)
  (CallFunction|HotfixableGEApplication (Variables|ScalableFloats|GetLavaPeriodicDamageEnabled) "/Game/Athena/Environments/Blueprints/SurfaceEffects/GE_SurfaceChange_Lava_Damage_Periodic.GE_SurfaceChange_Lava_Damage_Periodic_C")
  (Math|Random|SetRandomStreamSeed (Variables|Default|GetRandomAngleStream) (Math|Integer|int+int (Math|Float|Floor (/ _returnvalue_x 2e+02)) (Math|Float|Floor (/ _returnvalue_y 2e+02)) (Math|Float|Floor (Math|Rotator|NormalizeAxis _returnvalue_yaw))))
  (Variables|Default|SetRandomAngleOffset (* (ScalableFloat|GetValueatLevel (Variables|ScalableFloats|GetRandomConeAngle)) (- 0.5 (Math|Random|RandomFloatfromStream (Variables|Default|GetRandomAngleStream)))))
  (if (>= (Math|Vector|VectorLength _returnvalue_24) 5e+02)
    (Character|LaunchCharacter (Variables|Default|GetAvatarPawn) _returnvalue_24 true true)
    (Ability|ExecuteGameplayCueWithParamsOnOwner (Variables|Default|GetLavaBounceCue) (Ability|GameplayCue|MakeGameplayCueParameters :NormalizedMagnitude 0.0 :RawMagnitude 0.0 :EffectContext "()" :AggregatedSourceTags "(GameplayTags=)" :AggregatedTargetTags "(GameplayTags=)" :Location "0,0,0" :Normal (Math|Vector|MakeVectorNetQuantizeNormal (.x (Math|Vector|Normalize _returnvalue_24)) (.y (Math|Vector|Normalize _returnvalue_24)) (.z (Math|Vector|Normalize _returnvalue_24))) :Instigator (Variables|Default|GetAvatarPawn) :EffectCauser (Variables|Default|GetAvatarPawn) :GameplayEffectLevel 1 :AbilityLevel 1 :bReplicateLocationWhenUsingMinimalRepProxy false))
    (Variables|Default|SetLavaBounceMultiplier (Math|Float|Min(Float) (+ (Variables|Default|GetLavaBounceMultiplier) (ScalableFloat|GetValueatLevel (Variables|ScalableFloats|GetBounceMultiplierStepAmount))) (ScalableFloat|GetValueatLevel (Variables|ScalableFloats|GetMaxBounceMultiplier))))))

(fn OnRep_ReplicatedRandomAngle
  (Development|PrintString (Utilities|String|Append "On Rep Angle: " (Utilities|String|ToString(Float) (Variables|Default|GetRandomAngleOffset)))))

(fn ShouldApplyLavaPeriodicDamage
  (params (FortMovementComp :type Fort%20Movement%20Comp%20Character%20Object%20Reference) (FortSurfaceTrackingComp :type Fort%20Surface%20Tracking%20Component%20Object%20Reference))
  (if (Math|Boolean|ANDBoolean (Utilities|IsValid FortMovementComp :!class /Script/Engine.KismetSystemLibrary) (Utilities|IsValid FortSurfaceTrackingComp :!class /Script/Engine.KismetSystemLibrary) (Utilities|IsValid (Variables|Default|GetAvatarPawn) :!class /Script/Engine.KismetSystemLibrary))
    (return (Math|Boolean|ANDBoolean (Class|FortSurfaceTrackingComponent|IsComponentOwneronLava FortSurfaceTrackingComp false) (AI|Components|NavMovement|IsMovingonGround FortMovementComp) (not (GameplayTags|HasMatchingGameplayTag (Variables|Default|GetAvatarPawn) "(TagName=\"Granted.Athena.Surface.Lava.PeriodicDamage\")"))))
    (else
      (return false))))

(fn ShouldRemoveLavaPeriodicDamage
  (params (FortMovementComp :type Fort%20Movement%20Comp%20Character%20Object%20Reference) (FortSurfaceTrackingComp :type Fort%20Surface%20Tracking%20Component%20Object%20Reference))
  (if (Math|Boolean|ANDBoolean (Utilities|IsValid FortMovementComp :!class /Script/Engine.KismetSystemLibrary) (Utilities|IsValid FortSurfaceTrackingComp :!class /Script/Engine.KismetSystemLibrary) (Utilities|IsValid (Variables|Default|GetAvatarPawn) :!class /Script/Engine.KismetSystemLibrary))
    (return (and (or (not (Class|FortSurfaceTrackingComponent|IsComponentOwneronLava FortSurfaceTrackingComp false)) (not (AI|Components|NavMovement|IsMovingonGround FortMovementComp))) (GameplayTags|HasMatchingGameplayTag (Variables|Default|GetAvatarPawn) "(TagName=\"Granted.Athena.Surface.Lava.PeriodicDamage\")")))
    (else
      (return false))))

(fn SurfaceCleanup
  (params (AddLingeringEffects :type Boolean) (SurfaceType :type EPhysicalSurface%20Enum :default "SurfaceType_Default"))
  (bind _avatarpawn (Variables|Default|GetAvatarPawn))
  (bind _lavaperiodicdamagetags (Variables|Default|GetLavaPeriodicDamageTags))
  (bind _chrometags (Variables|Default|GetChromeTags))
  (bind _sandtags (Variables|Default|GetSandTags))
  (bind _snowtags (Variables|Default|GetSnowTags))
  (bind _lavatags (Variables|Default|GetLavaTags))
  (bind _icetags (Variables|Default|GetIceTags))
  (bind (_case_0_true _case_1_true _case_2_true _case_3_true _case_4_true _case_5_true) (GameplayTags|TagContainer|CompareTagAssetInterfacetoOtherTagContainers _icetags _lavatags _snowtags _sandtags _chrometags _lavaperiodicdamagetags _avatarpawn))
  (LocalVariables|SetbAddLingeringEffects AddLingeringEffects)
  (LocalVariables|SetbHasIce (GameplayTags|HasTag (Variables|Default|GetPreviousSurfaceTagContainer) "(TagName=\"Fort.SurfaceType.Ice\")"))
  (LocalVariables|SetbHasLava (or _case_1_true _case_5_true))
  (LocalVariables|SetbHasSnow _case_2_true)
  (LocalVariables|SetbHasSand _case_3_true)
  (LocalVariables|SetbHasChrome _case_4_true)
  (Utilities|FlowControl|Sequence
    (:then_0
      (if (and (LocalVariables|GetbHasIce) (Utilities|Enum|NotEqual(Enum) (Variables|GetSurfaceType) "SurfaceType24"))
        (if (and (> (ScalableFloat|GetValueatLevel (Variables|ScalableFloats|GetIceLingerDuration)) 0) (LocalVariables|GetbAddLingeringEffects))
          (Ability|ApplyGameplayEffectToOwner "/Game/Athena/Environments/Blueprints/SurfaceEffects/GE_SurfaceChange_Ice_Linger.GE_SurfaceChange_Ice_Linger_C" 0)
          (else
            ;[0.9531 0.9531 0.9531 0.5] Moved here from before the last Branch to fix FORT-1086644
            (Ability|RemoveGameplayEffectFromOwnerWithGrantedTags (Variables|Default|GetIceTags))
            ;[1 0 0] Workaround for no client predicted GE removal: This GE suppresses the icy surface GE for a short duration so that it is immediately disabled on the client, giving the server time to actually remove it
            (Ability|ApplyGameplayEffectToOwner "/Game/Athena/Environments/Blueprints/SurfaceEffects/GE_SurfaceChange_Ice_Suppress.GE_SurfaceChange_Ice_Suppress_C" 0)))))
    (:then_1
      (if (and (LocalVariables|GetbHasLava) (Utilities|Enum|NotEqual(Enum) (Variables|GetSurfaceType) "SurfaceType25"))
        (Ability|RemoveGameplayEffectFromOwnerWithGrantedTags (Variables|Default|GetLavaTags))
        (Ability|RemoveGameplayEffectFromOwnerWithGrantedTags (Variables|Default|GetLavaPeriodicDamageTags))
        (Variables|Default|SetLavaBounceMultiplier 1.0)))
    (:then_2
      (if (and (LocalVariables|GetbHasSnow) (Utilities|Enum|NotEqual(Enum) (Variables|GetSurfaceType) "SurfaceType23"))
        (Ability|RemoveGameplayEffectFromOwnerWithGrantedTags (Variables|Default|GetSnowTags))))
    (:then_3
      (if (and (LocalVariables|GetbHasSand) (Utilities|Enum|NotEqual(Enum) (Variables|GetSurfaceType) "SurfaceType26"))
        (Ability|SendGameplayEvent (Variables|Default|GetEventTag_SandUnburrow))
        (Ability|RemoveGameplayEffectFromOwnerWithGrantedTags (Variables|Default|GetSandTags))))
    (:then_4
      (if (and (LocalVariables|GetbHasChrome) (Utilities|Enum|NotEqual(Enum) (Variables|GetSurfaceType) "SurfaceType28"))
        (Ability|RemoveGameplayEffectFromOwnerWithGrantedTags (Variables|Default|GetChromeTags))))))

(fn SurfaceCleanupSpecial
  (Ability|RemoveGameplayEffectFromOwnerWithAssetTags "(GameplayTags=((TagName=\"Fort.SurfaceType\")))"))
