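; Gameplay ability that reacts to surface changes for the Octopus vehicle, triggered by an event whose instigator tags name the surface.
; It removes stale Ice or Lava effects, then applies the ice effect or multicasts a lava bounce.
; The bounce skips next fall damage and broadcasts an impact.
; Authority fires a cue with a direction blending mirrored velocity and randomized facing, then steps up a capped multiplier and rerolls the angle.
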(event Ability|EventActivateAbilityFromEvent
  (params (EventData :type Gameplay%20Event%20Data%20Structure%20%28by%20ref%29))
  (bind (_eventtag _instigator _target _optionalobject _optionalobject2 _contexthandle _instigatortags) (Utilities|Struct|BreakGameplayEventData EventData))
  (Development|PrintString (GameplayTags|GetDebugStringfromGameplayTagContainer _instigatortags) (Variables|Default|GetDebug) (Variables|Default|GetDebug))
  (switch Utilities|FlowControl|Switch|SwitchonGameplayTagContainer _instigatortags
    (:Default
      (CallFunction|SurfaceCleanup "SurfaceType_Default"))
    (:(GameplayTags=((TagName="Fort.SurfaceType.Ice")))
      (CallFunction|SurfaceCleanup "SurfaceType24")
      (Ability|ApplyGameplayEffectToOwner (Variables|Default|GetGE_SurfaceChange_Ice)))
    (:(GameplayTags=((TagName="Fort.SurfaceType.Lava")))
      (CallFunction|SurfaceCleanup "SurfaceType25")
      (CallFunction|MulticastLavaBounce (Ability|GetAvatarActorfromActorInfo))))
  (Ability|EndAbility))

(event Ability|EventOnEndAbility
  (params (bWasCancelled :type Boolean)))

(event Custom|MulticastLavaBounce
  (params (AvatarActor :type Object%20Reference))
  (CallFunction|LavaBounce AvatarActor))

(fn HotfixableGEApplication
  (params (Input :type Scalable%20Float%20Structure%20%28by%20ref%29) (GameplayEffectAppliedOnTrue :type Gameplay%20Effect%20Class%20Reference) (GameplayEffectAppliedOnFalse :type Gameplay%20Effect%20Class%20Reference))
  (if (> (ScalableFloat|GetValueatLevel Input) 0.0)
    (Ability|ApplyGameplayEffectToOwner GameplayEffectAppliedOnTrue 0)
    (else
      (Ability|ApplyGameplayEffectToOwner GameplayEffectAppliedOnFalse 0))))

(fn LavaBounce
  (params (AvatarActor :type Object%20Reference))
  (bind (_bistouching _contactpoint _contactnormal) (Class|FortOctopusVehicle|GetPrimaryContact (LocalVariables|GetOctopusVehicle)))
  (bind _asfort_octopus_vehicle (Utilities|Casting|CastToFortOctopusVehicle AvatarActor))
  (LocalVariables|SetOctopusVehicle _asfort_octopus_vehicle)
  (Vehicle|SetIgnoreNextFallingDamage _asfort_octopus_vehicle true)
  (Audio|BroadcastAbilityImpactatLocation 0 (LocalVariables|GetOctopusVehicle) (+ (Math|Vector|MakeVector (.x (Transformation|GetVelocity (LocalVariables|GetOctopusVehicle))) (.y (Transformation|GetVelocity (LocalVariables|GetOctopusVehicle))) 5e+02) (* _contactnormal 1e+02)) 775.0)
  (if (Networking|HasAuthority (LocalVariables|GetOctopusVehicle))
    (Ability|ExecuteGameplayCueWithParamsOnOwner (Variables|Default|GetLavaBounceCue) (Ability|GameplayCue|MakeGameplayCueParameters :NormalizedMagnitude 0.0 :RawMagnitude 0.0 :EffectContext "()" :AggregatedSourceTags "(GameplayTags=)" :AggregatedTargetTags "(GameplayTags=)" :Location "0,0,0" :Normal (Math|Vector|MakeVectorNetQuantizeNormal (.x (Math|Vector|Normalize (+ (* "0.000000,0.000000,1.000000" (ScalableFloat|GetValueatLevel (Variables|Default|GetBaseVerticalBounceVelocity))) (* (Math|Vector|Normalize (Math|Conversions|ToVector(Vector2D) (+ (* (Math|Vector2D|Normalize2D (Math|Conversions|ToVector2D(Vector) (Math|Vector|MirrorVectorbyNormal (Transformation|GetVelocity (LocalVariables|GetOctopusVehicle)) "0, 0,1.000000"))) (- 1.0 (ScalableFloat|GetValueatLevel (Variables|Default|GetFacingAngleVelocityComponent)))) (* (Math|Vector2D|Normalize2D (Math|Conversions|ToVector2D(Vector) (Math|Vector|RotateVector (Math|Rotator|GetRotationXVector (Transformation|GetActorRotation (LocalVariables|GetOctopusVehicle))) (Math|Rotator|MakeRotator 0.0 0.0 (Variables|Default|GetReplicatedRandomAngle))))) (ScalableFloat|GetValueatLevel (Variables|Default|GetFacingAngleVelocityComponent)))))) (* (Variables|Default|GetLavaBounceMultiplier) (ScalableFloat|GetValueatLevel (Variables|Default|GetBaseLateralBounceVelocity))))))) (.y (Math|Vector|Normalize (+ (* "0.000000,0.000000,1.000000" (ScalableFloat|GetValueatLevel (Variables|Default|GetBaseVerticalBounceVelocity))) (* (Math|Vector|Normalize (Math|Conversions|ToVector(Vector2D) (+ (* (Math|Vector2D|Normalize2D (Math|Conversions|ToVector2D(Vector) (Math|Vector|MirrorVectorbyNormal (Transformation|GetVelocity (LocalVariables|GetOctopusVehicle)) "0, 0,1.000000"))) (- 1.0 (ScalableFloat|GetValueatLevel (Variables|Default|GetFacingAngleVelocityComponent)))) (* (Math|Vector2D|Normalize2D (Math|Conversions|ToVector2D(Vector) (Math|Vector|RotateVector (Math|Rotator|GetRotationXVector (Transformation|GetActorRotation (LocalVariables|GetOctopusVehicle))) (Math|Rotator|MakeRotator 0.0 0.0 (Variables|Default|GetReplicatedRandomAngle))))) (ScalableFloat|GetValueatLevel (Variables|Default|GetFacingAngleVelocityComponent)))))) (* (Variables|Default|GetLavaBounceMultiplier) (ScalableFloat|GetValueatLevel (Variables|Default|GetBaseLateralBounceVelocity))))))) (.z (Math|Vector|Normalize (+ (* "0.000000,0.000000,1.000000" (ScalableFloat|GetValueatLevel (Variables|Default|GetBaseVerticalBounceVelocity))) (* (Math|Vector|Normalize (Math|Conversions|ToVector(Vector2D) (+ (* (Math|Vector2D|Normalize2D (Math|Conversions|ToVector2D(Vector) (Math|Vector|MirrorVectorbyNormal (Transformation|GetVelocity (LocalVariables|GetOctopusVehicle)) "0, 0,1.000000"))) (- 1.0 (ScalableFloat|GetValueatLevel (Variables|Default|GetFacingAngleVelocityComponent)))) (* (Math|Vector2D|Normalize2D (Math|Conversions|ToVector2D(Vector) (Math|Vector|RotateVector (Math|Rotator|GetRotationXVector (Transformation|GetActorRotation (LocalVariables|GetOctopusVehicle))) (Math|Rotator|MakeRotator 0.0 0.0 (Variables|Default|GetReplicatedRandomAngle))))) (ScalableFloat|GetValueatLevel (Variables|Default|GetFacingAngleVelocityComponent)))))) (* (Variables|Default|GetLavaBounceMultiplier) (ScalableFloat|GetValueatLevel (Variables|Default|GetBaseLateralBounceVelocity)))))))) :Instigator (LocalVariables|GetOctopusVehicle) :EffectCauser (LocalVariables|GetOctopusVehicle) :GameplayEffectLevel 1 :AbilityLevel 1 :bReplicateLocationWhenUsingMinimalRepProxy false))
    (Variables|Default|SetLavaBounceMultiplier (Math|Float|Min(Float) (+ (Variables|Default|GetLavaBounceMultiplier) (ScalableFloat|GetValueatLevel (Variables|Default|GetBounceMultiplierStepAmount))) (ScalableFloat|GetValueatLevel (Variables|Default|GetMaxBounceMultiplier))))
    (Variables|Default|SetReplicatedRandomAngle (Math|Random|RandomFloatinRange (- 0.0 (ScalableFloat|GetValueatLevel (Variables|Default|GetRandomConeAngle))) (ScalableFloat|GetValueatLevel (Variables|Default|GetRandomConeAngle))))))

(fn RemoveGameplayEffects
  (Ability|RemoveGameplayEffectFromOwnerWithAssetTags "(GameplayTags=((TagName=\"Fort.SurfaceType\")))"))

(fn SurfaceCleanup
  (params (SurfaceType :type EPhysicalSurface%20Enum :default "SurfaceType_Default"))
  (bind _icetags (Variables|Default|GetIceTags))
  (bind _lavatags (Variables|Default|GetLavaTags))
  (bind _asfort_octopus_vehicle (Utilities|Casting|CastToFortOctopusVehicle (Ability|GetAvatarActorfromActorInfo)))
  (LocalVariables|SetbHasIce (GameplayTags|TagContainer|CompareTagAssetInterfacetoOtherTagContainers _icetags _lavatags _asfort_octopus_vehicle))
  (LocalVariables|SetbHasLava (.Case_1%20True (GameplayTags|TagContainer|CompareTagAssetInterfacetoOtherTagContainers _icetags _lavatags _asfort_octopus_vehicle)))
  (Utilities|FlowControl|Sequence
    (:then_0
      (if (and (LocalVariables|GetbHasIce) (Utilities|Enum|NotEqual(Enum) (Variables|GetSurfaceType) "SurfaceType24"))
        (Ability|RemoveGameplayEffectFromOwnerWithGrantedTags (Variables|Default|GetIceTags))))
    (:then_1
      (if (and (LocalVariables|GetbHasLava) (Utilities|Enum|NotEqual(Enum) (Variables|GetSurfaceType) "SurfaceType25"))
        (Ability|RemoveGameplayEffectFromOwnerWithGrantedTags (Variables|Default|GetLavaTags))
        (Variables|Default|SetLavaBounceMultiplier 1.0)))))
