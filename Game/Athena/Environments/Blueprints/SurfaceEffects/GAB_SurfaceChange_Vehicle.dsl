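; Gameplay ability that reacts to surface-change events on vehicles, applying ice or lava effects.
; It switches on instigator surface tags, clearing stale effects; ice skips Spaghetti, Octopus and SKPush vehicles and cannons.
; Lava applies damage and multicasts a bounce blending mirrored velocity with randomly angled facing, ignoring the next fall damage.
; The authority fires the cue, steps up a capped bounce multiplier (reset on leaving lava) and rerolls the replicated angle.
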(event Ability|EventActivateAbilityFromEvent
  (params (EventData :type Gameplay%20Event%20Data%20Structure%20%28by%20ref%29))
  (bind (_eventtag _instigator _target _optionalobject _optionalobject2 _contexthandle _instigatortags) (Utilities|Struct|BreakGameplayEventData EventData))
  (bind (_asfort_spaghetti_vehicle _bsuccess) (Utilities|Casting|CastToFortSpaghettiVehicle (Ability|GetAvatarActorfromActorInfo)))
  (bind (_asfort_athena_skpush_cannon _bsuccess_2) (Utilities|Casting|CastToFortAthenaSKPushCannon (Ability|GetAvatarActorfromActorInfo)))
  (bind (_asfort_octopus_vehicle _bsuccess_3) (Utilities|Casting|CastToFortOctopusVehicle (Ability|GetAvatarActorfromActorInfo)))
  (bind (_asfort_athena_skpush_vehicle _bsuccess_4) (Utilities|Casting|CastToFortAthenaSKPushVehicle (Ability|GetAvatarActorfromActorInfo)))
  (Development|PrintString (GameplayTags|GetDebugStringfromGameplayTagContainer _instigatortags) (Variables|Default|GetDebug) (Variables|Default|GetDebug))
  (switch Utilities|FlowControl|Switch|SwitchonGameplayTagContainer _instigatortags
    (:Default
      (CallFunction|SurfaceCleanup "SurfaceType_Default"))
    (:(GameplayTags=((TagName="Fort.SurfaceType.Ice")))
      (if (not (Math|Boolean|ORBoolean _bsuccess_2 _bsuccess_3 _bsuccess _bsuccess_4))
        (CallFunction|SurfaceCleanup "SurfaceType24")
        (Ability|ApplyGameplayEffectToOwner (Variables|Default|GetGE_SurfaceChange_Ice))))
    (:(GameplayTags=((TagName="Fort.SurfaceType.Lava")))
      (CallFunction|SurfaceCleanup "SurfaceType25")
      (CallFunction|MulticastLavaBounce (Ability|GetAvatarActorfromActorInfo))
      (Ability|ApplyGameplayEffectToOwner "/Game/Athena/Environments/Blueprints/SurfaceEffects/GE_SurfaceChange_Lava_Damage.GE_SurfaceChange_Lava_Damage_C" 0)))
  (Ability|EndAbility))

(event Ability|EventOnEndAbility
  (params (bWasCancelled :type Boolean)))

(event Custom|MulticastLavaBounce
  (params (AvatarActor :type Object%20Reference))
  (CallFunction|LavaBounce AvatarActor))

(fn HotfixableGEApplication
  (params (Input :type Scalable%20Float%20Structure%20%28by%20ref%29) (GameplayEffectAppliedOnTrue :type Gameplay%20Effect%20Class%20Reference) (GameplayEffectAppliedOnFalse :type Gameplay%20Effect%20Class%20Reference))
  (if (> (ScalableFloat|GetValueatLevel Input) 0.0)
    (Ability|ApplyGameplayEffectToOwner GameplayEffectAppliedOnTrue 0)
    (else
      (Ability|ApplyGameplayEffectToOwner GameplayEffectAppliedOnFalse 0))))

(fn LavaBounce
  (params (AvatarActor :type Object%20Reference))
  (bind _baseverticalbouncevelocity (Variables|Default|GetBaseVerticalBounceVelocity))
  (bind _returnvalue (ScalableFloat|GetValueatLevel _baseverticalbouncevelocity))
  (bind _returnvalue_2 (* "0.000000,0.000000,1.000000" _returnvalue))
  (bind _baselateralbouncevelocity (Variables|Default|GetBaseLateralBounceVelocity))
  (bind _returnvalue_3 (ScalableFloat|GetValueatLevel _baselateralbouncevelocity))
  (bind _lavabouncemultiplier (Variables|Default|GetLavaBounceMultiplier))
  (bind _returnvalue_4 (* _lavabouncemultiplier _returnvalue_3))
  (bind _facinganglevelocitycomponent (Variables|Default|GetFacingAngleVelocityComponent))
  (bind _returnvalue_5 (ScalableFloat|GetValueatLevel _facinganglevelocitycomponent))
  (bind _replicatedrandomangle (Variables|Default|GetReplicatedRandomAngle))
  (bind _returnvalue_6 (Math|Rotator|MakeRotator 0.0 0.0 _replicatedrandomangle))
  (bind _vehicleinterface (LocalVariables|GetVehicleInterface))
  (bind _returnvalue_7 (Vehicle|GetVehicleRotation _vehicleinterface))
  (bind _returnvalue_8 (Math|Rotator|GetRotationXVector _returnvalue_7))
  (bind _returnvalue_9 (Math|Vector|RotateVector _returnvalue_8 _returnvalue_6))
  (bind _returnvalue_10 (Math|Conversions|ToVector2D(Vector) _returnvalue_9))
  (bind _returnvalue_11 (Math|Vector2D|Normalize2D _returnvalue_10))
  (bind _returnvalue_12 (* _returnvalue_11 _returnvalue_5))
  (bind _returnvalue_13 (- 1.0 _returnvalue_5))
  (bind _vehicleinterface_2 (LocalVariables|GetVehicleInterface))
  (bind _returnvalue_14 (Vehicle|GetVehicleVelocity _vehicleinterface_2))
  (bind _returnvalue_15 (Math|Vector|MirrorVectorbyNormal _returnvalue_14 "0, 0,1.000000"))
  (bind _returnvalue_16 (Math|Conversions|ToVector2D(Vector) _returnvalue_15))
  (bind _returnvalue_17 (Math|Vector2D|Normalize2D _returnvalue_16))
  (bind _returnvalue_18 (* _returnvalue_17 _returnvalue_13))
  (bind _returnvalue_19 (+ _returnvalue_18 _returnvalue_12))
  (bind _returnvalue_20 (Math|Conversions|ToVector(Vector2D) _returnvalue_19))
  (bind _returnvalue_21 (Math|Vector|Normalize _returnvalue_20))
  (bind _returnvalue_22 (* _returnvalue_21 _returnvalue_4))
  (bind _returnvalue_23 (+ _returnvalue_2 _returnvalue_22))
  (Utilities|IsValid AvatarActor
    (:Is%20Valid
      (bind _asfort_vehicle_interface (Utilities|Casting|CastToFortVehicleInterface AvatarActor))
      (bind _output_get (LocalVariables|SetVehicleInterface _asfort_vehicle_interface))
      (Vehicle|SetIgnoreNextFallingDamage _output_get true)
      (Audio|BroadcastAbilityImpactatLocation 0 (Utilities|ToObject(Interface) (LocalVariables|GetVehicleInterface)) _returnvalue_23 775.0)
      (bind _returnvalue_24 (Vehicle|GetVehicleActor (LocalVariables|GetVehicleInterface) :!class /Script/FortniteGame.FortVehicleInterface))
      (if (Networking|HasAuthority _returnvalue_24)
        (Ability|ExecuteGameplayCueWithParamsOnOwner (Variables|Default|GetLavaBounceCue) (Ability|GameplayCue|MakeGameplayCueParameters :NormalizedMagnitude 0.0 :RawMagnitude 0.0 :EffectContext "()" :AggregatedSourceTags "(GameplayTags=)" :AggregatedTargetTags "(GameplayTags=)" :Location "0,0,0" :Normal (Math|Vector|MakeVectorNetQuantizeNormal (.x (Math|Vector|Normalize _returnvalue_23)) (.y (Math|Vector|Normalize _returnvalue_23)) (.z (Math|Vector|Normalize _returnvalue_23))) :Instigator _returnvalue_24 :EffectCauser _returnvalue_24 :GameplayEffectLevel 1 :AbilityLevel 1 :bReplicateLocationWhenUsingMinimalRepProxy false))
        (Variables|Default|SetLavaBounceMultiplier (Math|Float|Min(Float) (+ (Variables|Default|GetLavaBounceMultiplier) (ScalableFloat|GetValueatLevel (Variables|Default|GetBounceMultiplierStepAmount))) (ScalableFloat|GetValueatLevel (Variables|Default|GetMaxBounceMultiplier))))
        (Variables|Default|SetReplicatedRandomAngle (Math|Random|RandomFloatinRange (- 0.0 (ScalableFloat|GetValueatLevel (Variables|Default|GetRandomConeAngle))) (ScalableFloat|GetValueatLevel (Variables|Default|GetRandomConeAngle))))))))

(fn RemoveGameplayEffects
  (Ability|RemoveGameplayEffectFromOwnerWithAssetTags "(GameplayTags=((TagName=\"Fort.SurfaceType\")))"))

(fn SurfaceCleanup
  (params (SurfaceType :type EPhysicalSurface%20Enum :default "SurfaceType_Default"))
  (bind _returnvalue (Ability|GetAvatarActorfromActorInfo))
  (bind _icetags (Variables|Default|GetIceTags))
  (bind _icetags_2 (Variables|Default|GetIceTags))
  (bind _lavatags (Variables|Default|GetLavaTags))
  (bind _lavatags_2 (Variables|Default|GetLavaTags))
  ;[1 1 1] FortAthenaVehicle
  (bind _asfort_vehicle_interface (Utilities|Casting|CastToFortVehicleInterface _returnvalue)
    (:then
      (bind _asactor (Utilities|Casting|CastToActor _asfort_vehicle_interface))
      (bind _asgameplay_tag_asset_interface (Utilities|Casting|CastToGameplayTagAssetInterface _asactor))
      (LocalVariables|SetbHasIce (GameplayTags|TagContainer|CompareTagAssetInterfacetoOtherTagContainers _icetags _lavatags_2 _asgameplay_tag_asset_interface))
      (LocalVariables|SetbHasLava (.Case_1%20True (GameplayTags|TagContainer|CompareTagAssetInterfacetoOtherTagContainers _icetags _lavatags_2 _asgameplay_tag_asset_interface))))
    (:CastFailed
      ;[1 1 1] CharacterVehicle
      (bind _asfort_character_vehicle (Utilities|Casting|CastToFortCharacterVehicle _returnvalue))
      (LocalVariables|SetbHasIce (GameplayTags|TagContainer|CompareTagAssetInterfacetoOtherTagContainers _icetags_2 _lavatags _asfort_character_vehicle))
      (LocalVariables|SetbHasLava (.Case_1%20True (GameplayTags|TagContainer|CompareTagAssetInterfacetoOtherTagContainers _icetags_2 _lavatags _asfort_character_vehicle)))))
  (Utilities|FlowControl|Sequence
    (:then_0
      (if (and (LocalVariables|GetbHasIce) (Utilities|Enum|NotEqual(Enum) (Variables|GetSurfaceType) "SurfaceType24"))
        (Ability|RemoveGameplayEffectFromOwnerWithGrantedTags (Variables|Default|GetIceTags))))
    (:then_1
      (if (and (LocalVariables|GetbHasLava) (Utilities|Enum|NotEqual(Enum) (Variables|GetSurfaceType) "SurfaceType25"))
        (Ability|RemoveGameplayEffectFromOwnerWithGrantedTags (Variables|Default|GetLavaTags))
        (Variables|Default|SetLavaBounceMultiplier 1.0)))))
