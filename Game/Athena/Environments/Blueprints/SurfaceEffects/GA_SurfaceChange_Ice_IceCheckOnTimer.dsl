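; Gameplay ability that removes a player's icy-feet surface effect once they have been off ice for a linger duration.
; A looping timer, its interval clamped from that duration, traces the terrain: non-ice starts an off-ice timer if none runs, and ice cancels it.
; On expiry it traces again, and non-ice or no hit removes the ice effects by asset and granted tags and ends the ability, clearing both timers.
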(event Ability|EventActivateAbility
  (bind _asfort_player_pawn (Utilities|Casting|CastToFortPlayerPawn (Ability|GetAvatarActorfromActorInfo)))
  (Variables|Default|SetFortPlayerPawn _asfort_player_pawn)
  (bind _returnvalue (Utilities|Time|SetTimerbyEvent (delegate Custom|TraceForTerrain) (Math|Float|Clamp(Float) (ScalableFloat|GetValueatLevel (Variables|Default|GetIceLingerDuration)) 0.1) true))
  (Variables|Default|SetTraceTimer _returnvalue))

(event Custom|TraceForTerrain
  (Utilities|TraceCheck
    (:HitNonIce
      (if (not (Utilities|Time|IsTimerActivebyHandle (Variables|Default|GetOffIceTimer)))
        (Utilities|Time|ClearandInvalidateTimerbyHandle (Variables|Default|GetOffIceTimer))
        (bind _returnvalue (Utilities|Time|SetTimerbyEvent (delegate Custom|RemoveIceGE) (ScalableFloat|GetValueatLevel (Variables|Default|GetIceLingerDuration))))
        (Variables|Default|SetOffIceTimer _returnvalue)))
    (:HitIce
      (Utilities|Time|ClearandInvalidateTimerbyHandle (Variables|Default|GetOffIceTimer)))))

(event Custom|RemoveIceGE
  (Utilities|TraceCheck
    (:HitNonIce
      _)
    (:NoHit
      _))
  (Ability|RemoveGameplayEffectFromOwnerWithAssetTags "(GameplayTags=((TagName=\"GameplayEffect.Surface.Ice\")))")
  ; Moved here from GAB_SurfaceChange/SurfaceCleanup to fix FORT-1086644
  (Ability|RemoveGameplayEffectFromOwnerWithGrantedTags "(GameplayTags=((TagName=\"Gameplay.Status.IcyFeet.Surface\")))")
  (Ability|EndAbility))

(event Ability|EventOnEndAbility
  (params (bWasCancelled :type Boolean))
  (Utilities|Time|ClearandInvalidateTimerbyHandle (Variables|Default|GetTraceTimer))
  (Utilities|Time|ClearandInvalidateTimerbyHandle (Variables|Default|GetOffIceTimer)))
